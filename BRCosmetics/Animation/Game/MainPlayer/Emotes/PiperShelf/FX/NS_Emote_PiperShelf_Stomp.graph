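# Niagara burst effect for the stomp in the PiperShelf emote, placed on the skeletal mesh at the bone given by User.BoneIndex.
# Bright_Flash_Decal is a hue-cycling sprite and fading light that tracks that bone.
# Light_Shafts adds brief streaks scaled by velocity direction, while Sparkles_Glint throws curl-noise, drag-slowed glints drawn by three sprite renderers.
# Time-of-day inputs adjust flash brightness and tint shafts and sparkles blue at night.
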NiagaraSystem /BRCosmetics/Animation/Game/MainPlayer/Emotes/PiperShelf/FX/NS_Emote_PiperShelf_Stomp
dependencies {
 usedDataInterfaces [
  &NiagaraDataInterfaceCurve
  &NiagaraDataInterfaceArrayFloat
  &NiagaraDataInterfaceColorCurve
  &NiagaraDataInterfaceSkeletalMesh
 ]
 usedDynamicInputs [
  &ReturnNormalizedExecIndex
  &NormalizedAngleToDegrees
  &AddVectorToPosition
 ]
 usedModules [
  &self:SystemSpawnScript.SystemScriptSource.SystemScriptGraph.NiagaraNodeAssignment_0.SetVariables_#1
  &SystemState
 ]
}
emitters {
 Bright_Flash_Decal {
  emitterUpdateScript [
   &EmitterState()
   &SpawnBurst_Instantaneous(Spawn Count=1, Use Spawn Probability=false)
  ]
  particleSpawnScript [
   &InitializeParticle(
    Color=&MultiplyLinearColor(
     Linear Color=<rgba 1 0.9 0.9993 0.75>
     Scale Factor=&MakeVector4_FromFloat(Float=&FNTimeOfDayScalar(Night Scale=5, Morning Scale=7, Day Scale=10, Evening Scale=7))
    )
    Lifetime=0.5
    Sprite Rotation Mode=::NewEnumerator2
    Sprite Size=&Vector2DFromFloat(Value=30)
    Sprite Size Mode=::NewEnumerator4
    UsePositionOffset=false
    Write Material Random=true
   ) {module=InitializeParticle001}
   &CameraOffset(Camera Offset Amount=10)
  ]
  particleUpdateScript [
   &ParticleState()
   &SkeletalMeshLocation(
    Bone / Socket Index=&MakeIntFromFloat(Float=$User.BoneIndex)
    Bone Sampling Mode=::NewEnumerator4
    MaskSpawnGroup=false
    Sampled Position Offset=&MakeVector(X=0, Y=0, Z=3)
    Sampled Position Offset Coordinate Space:ENiagaraSkelMeshTransforms=::NewEnumerator1
    Skeletal Mesh=$System.Skeletal Mesh
   ) {module=SkeletalMeshLocation001}
   &ScaleColor(
    Scale Alpha=&FloatFromCurve(FloatCurve=DataInterface() {propertyValues {Curve {keys [
         {arriveTangent=0.5269, interpMode=RCIM_Cubic, leaveTangent=0.5269, tangentMode=RCTM_User, time=0.02, value=0.9674}
         {arriveTangent=-0.2619, interpMode=RCIM_Cubic, leaveTangent=-0.2619, tangentMode=RCTM_User, time=0.2607, value=0.9753}
         <at 1 0 RCIM_Cubic>
        ]}}})
    Scale RGB=&MakeVectorFromLinearColorRGB(Color=&ColorFromCurve(
      ColorCurve=DataInterface() {propertyValues {
        AlphaCurve {keys [<at 0 1>,<at 1 1>]}
        BlueCurve {keys [<at 0 0.65>,<at 0.333 0.65>,<at 0.667 1>,<at 1 0.65>]}
        GreenCurve {keys [<at 0 0.65>,<at 0.333 1>,<at 0.667 0.65>,<at 1 0.79>]}
        RedCurve {keys [<at 0 1>,<at 0.333 0.65>,<at 0.667 0.65>,<at 1 1>]}
       }}
      CurveIndex=&Multiply_Float(A=$Particles.NormalizedAge, B=2)
     ))
    ScaleA=true
    ScaleRGB=true
   )
   &SolveForcesAndVelocity(Clamp Velocity=false, Limit Acceleration=false)
   &DynamicMaterialParameters(
    Index 0 Param 1=&FloatFromCurve(FloatCurve=DataInterface() {propertyValues {Curve {keys [<at 0 5 RCIM_Cubic>,<at 0.0296 3 RCIM_Cubic>,<at 1 5 RCIM_Cubic>]}}})
    Index 0 Param 2=1.25
    Index 0 Param 3=&FloatFromCurve(FloatCurve=DataInterface() {propertyValues {Curve {keys [<at 0 0.35>,<at 1 0>]}}})
    Index 0 Param 4=&RandomRangeFloat(Minimum=0, Maximum=1)
    Index 1 Param 1=75
    Index 1 Param 2=5
    Module.Param0WriteEnabled=true
    Module.Param0WriteEnabled1=true
    Module.Param1WriteEnabled=true
    Module.Param1WriteEnabled1=true
    Module.Param2WriteEnabled=true
    Module.Param2WriteEnabled1=true
    Module.Param3WriteEnabled=true
    Module.Param3WriteEnabled1=true
    Write Parameter Index 1=true
   )
   &FN_Light_Attributes(
    Color=$Particles.Color
    Create Color Attribute=true
    Create Radius Attribute=true
    Radius=12
    Scale Color=&FloatFromCurve(
     FloatCurve=DataInterface() {propertyValues {Curve {keys [{interpMode=RCIM_Cubic, tangentMode=RCTM_User, time=0, value=1}, {interpMode=RCIM_Cubic, tangentMode=RCTM_User, time=1, value=0}]}}}
     Scale Curve=5
    )
   )
  ]
  rendererClasses [&NiagaraSpriteRendererProperties, &NiagaraLightRendererProperties]
  renderers [
   {rendererClass=&NiagaraSpriteRendererProperties}
   {rendererClass=&NiagaraLightRendererProperties, rendererIndex=1}
  ]
 }
 Light_Shafts {
  emitterUpdateScript [
   &EmitterState()
   &SpawnBurst_Instantaneous(Spawn Count=6, Use Spawn Probability=false, UseLoopCountLimit=false)
  ]
  particleSpawnScript [
   &PartitionParticles(Number Of Partitions=14)
   &InitializeParticle(
    Color Maximum=<rgba 1.984 2 1.7205 0.075>
    Color Minimum=<rgba 0.9732 0.9263 1 0.075>
    Color Mode=::NewEnumerator2
    Lifetime Max=0.3
    Lifetime Min=0.28
    Lifetime Mode=::NewEnumerator1
    Sprite Size Max=<50 15>
    Sprite Size Min=<30 10>
    Sprite Size Mode=::NewEnumerator2
    UsePositionOffset=false
    Write Material Random=true
   )
   /Niagara/Modules/Update/SubUV/SubUVAnimation(SubUV Animation Mode=::NewEnumerator1)
   &SkeletalMeshLocation(
    Bone / Socket Index=&MakeIntFromFloat(Float=$User.BoneIndex)
    MaskSpawnGroup=false
    Mesh Sampling Type=::NewEnumerator5
    Skeletal Mesh=$System.Skeletal Mesh
    Socket Sampling Mode=::NewEnumerator5
   )
   &ShapeLocation(MaskSpawnGroup=false, Override Seed=false, Sphere Radius=1.5, UseOwnerScale=true)
   &AddVelocity(
    Module.Use Falloff Distance=false
    Origin Offset=<0 0 -10>
    Override Seed=false
    Velocity Mode=::NewEnumerator1
    Velocity Origin=$Particles.SkeletalMeshLocation.SampledPosition
    Velocity Speed=5
   )
   &DynamicMaterialParameters(
    Index 0 Param 1=&UniformRangedFloat(Minimum=0, Maximum=1)
    Module.Param0WriteEnabled=true
    Module.Param1WriteEnabled=false
    Module.Param2WriteEnabled=false
    Module.Param3WriteEnabled=false
   )
   &CameraOffset(Camera Offset Amount=10)
  ]
  particleUpdateScript [
   &ParticleState()
   &ScaleColor(
    Scale Alpha=&FloatFromCurve(
     FloatCurve=DataInterface() {propertyValues {Curve {keys [
         {arriveTangent=0.013, interpMode=RCIM_Cubic, leaveTangent=0.013, tangentMode=RCTM_User, time=0.5, value=1}
         {tangentMode=RCTM_User, time=1, value=0}
        ]}}}
     Scale Curve=2.5
    )
    Scale RGB=&MakeVectorFromLinearColorRGB(Color=&FNTimeOfDayColor(Night Color=<rgba 0.15 0.575 1 1>, Morning Color=<rgba 1 1 1 1>, Day Color=<rgba 1 1 1 1>, Evening Color=<rgba 1 1 1 1>))
    ScaleA=true
    ScaleRGB=true
   )
   &ScaleSpriteSize(Scale Factor=&MakeVector2D(
     X=&FloatFromCurve(FloatCurve=DataInterface() {propertyValues {Curve {keys [
          {arriveTangent=0.0295, interpMode=RCIM_Cubic, leaveTangent=0.0295, tangentMode=RCTM_User, time=0, value=1}
          {arriveTangent=0.8671, interpMode=RCIM_Cubic, leaveTangent=0.8671, tangentMode=RCTM_User, time=1, value=1.45}
         ]}}})
     Y=&FloatFromCurve(
      FloatCurve=DataInterface() {propertyValues {Curve {keys [
          {arriveTangent=9.0725, interpMode=RCIM_Cubic, leaveTangent=9.0725, tangentMode=RCTM_User, time=0, value=0}
          {arriveTangent=0.0071, interpMode=RCIM_Cubic, leaveTangent=0.0071, tangentMode=RCTM_User, time=0.1408, value=1}
          {arriveTangent=-0.448, interpMode=RCIM_Cubic, leaveTangent=-0.448, tangentMode=RCTM_User, time=1, value=0.8}
         ]}}}
      Scale Curve=&Multiply_Float(
       A=4
       B=&Power(
        a=&MakeFloatFromVector(
         Vector=&NormalizeVector(Vector To Normalize=$Particles.Velocity, Safe Normalize=true, Safe Fallback Threshold=0, Safe Fallback Vector=<0 0 0.001>)
         Channel=::NewEnumerator2
        )
        b=4
       )
      )
     )
    ))
  ]
  rendererClasses [&NiagaraSpriteRendererProperties]
  renderers [{rendererClass=&NiagaraSpriteRendererProperties}]
 }
 Sparkles_Glint {
  emitterUpdateScript [
   &EmitterState()
   &SpawnBurst_Instantaneous(
    Spawn Count=&RandomRangeInt(Minimum=9, Maximum=16)
    Use Spawn Probability=false
    UseLoopCountLimit=false
   )
  ]
  particleSpawnScript [
   /SetVariables(Particles.VisibilityTag=&RandomRangeInt(Minimum=0, Maximum=2)) {isSetParameters=true}
   &InitializeParticle(
    AdjustAlpha=false
    AdjustHue=true
    AdjustSaturation=false
    AdjustValue=false
    Color=<rgba 0.843 0.8 1 1>
    Color Mode=::NewEnumerator3
    Hue Shift Range:Vector2f=<-1 1>
    Lifetime Max=0.5
    Lifetime Min=0.4
    Lifetime Mode=::NewEnumerator1
    Sprite Size Mode=::NewEnumerator3
    Uniform Sprite Size=&Multiply_Float(
     A=&Multiply_Float(
      A=&SelectFloatFromArray(Direct Array Index=$Particles.VisibilityTag, Float Selection Array=DataInterface() {propertyValues {FloatData [60,30,15]}})
      B=&RandomRangeFloat(Minimum=0.8, Maximum=1.2)
     )
     B=1
    )
    UsePositionOffset=false
    Write Material Random=true
   )
   &CameraOffset(Camera Offset Amount=10)
   &SkeletalMeshLocation(
    Bone / Socket Index=&MakeIntFromFloat(Float=$User.BoneIndex)
    MaskSpawnGroup=false
    Mesh Sampling Type=::NewEnumerator5
    Skeletal Mesh=$System.Skeletal Mesh
    Socket Sampling Mode=::NewEnumerator5
   ) {module=SkeletalMeshLocation001}
   &ShapeLocation(MaskSpawnGroup=false, Offset Mode=::NewEnumerator0, Override Seed=false, Sphere Radius=3, UseOwnerScale=true)
   &AddVelocity(
    Cone Angle=60
    Cone Axis=<0 0 1>
    Module.UseConeFalloff=true
    Module.UseVelDistribution=true
    Override Seed=false
    Velocity Mode=::NewEnumerator2
    Velocity Speed=&RandomRangeFloat(Minimum=100, Maximum=400)
   )
  ]
  particleUpdateScript [
   &ParticleState()
   &ScaleColor(
    Scale Alpha=&Multiply_Float(
     A=&FloatFromCurve(
      FloatCurve=DataInterface() {propertyValues {Curve {keys [
          {arriveTangent=9.5826, interpMode=RCIM_Cubic, leaveTangent=9.5826, tangentMode=RCTM_User, time=0, value=0}
          {arriveTangent=-2.4058, interpMode=RCIM_Cubic, leaveTangent=-2.4058, tangentMode=RCTM_User, time=0.26, value=0.8421}
          {arriveTangent=0.0151, interpMode=RCIM_Cubic, leaveTangent=0.0151, tangentMode=RCTM_User, time=0.5, value=0.4}
          {arriveTangent=-0.8, interpMode=RCIM_Cubic, leaveTangent=-0.8, time=0.8, value=0.6}
          <at 1 0 RCIM_Cubic>
         ]}}}
      Scale Curve=&RandomRangeFloat(Minimum=0.5, Maximum=1)
     )
     B=&FloatFromCurve(FloatCurve=DataInterface() {propertyValues {Curve {keys [{interpMode=RCIM_Cubic, tangentMode=RCTM_User, time=0, value=1}, {interpMode=RCIM_Cubic, tangentMode=RCTM_User, time=1, value=0}]}}})
    )
    Scale RGB=&MakeVectorFromLinearColorRGB(Color=&FNTimeOfDayColor(Night Color=<rgba 0.3 0.6267 1 1>, Morning Color=<rgba 1 1 1 1>, Day Color=<rgba 1 1 1 1>, Evening Color=<rgba 1 1 1 1>))
    ScaleA=true
    ScaleRGB=true
   )
   &ScaleSpriteSize(Scale Factor=&Multiply_Vector2D_ByFloat(
     Float=1
     Vector2D=&Vector2DFromFloat(Value=&FloatFromCurve(FloatCurve=DataInterface() {propertyValues {Curve {keys [{interpMode=RCIM_Cubic, tangentMode=RCTM_User, time=0, value=0}, {interpMode=RCIM_Cubic, tangentMode=RCTM_User, time=1, value=1}]}}}))
    ))
   /SetVariables(
    Particles.Color=&MultiplyLinearColor(
     Linear Color=$Particles.Color
     Scale Factor=&MakeVector4_FromLinearColor(LinearColor=<rgba 8.0099 9.8007 10 1>)
    )
    Particles.Color02:LinearColor=&MultiplyLinearColor(
     Linear Color=$Particles.Color
     Scale Factor=&MakeVector4_FromLinearColor(LinearColor=<rgba 0.9441 0.8079 1 1>)
    )
    Particles.Color03:LinearColor=&MultiplyLinearColor(
     Linear Color=$Particles.Color
     Scale Factor=&MakeVector4_FromLinearColor(LinearColor=<rgba 0.15 0.15 0.15 1>)
    )
    Particles.SpriteSize02:Vector2f=&Multiply_Vector2(A=$Particles.SpriteSize, B=&Vector2DFromFloat(Value=0.3))
   ) {isSetParameters=true}
   &CurlNoiseForce(Noise Frequency=100, Noise Strength=100, Pan Noise=false)
   &Drag(Drag=&RandomRangeFloat(Minimum=5, Maximum=10), Use Linear Drag=true, Use Rotational Drag=true)
   &SolveForcesAndVelocity(Clamp Velocity=false, Limit Acceleration=false)
  ]
  rendererClasses [&NiagaraSpriteRendererProperties]
  renderers [
   {rendererClass=&NiagaraSpriteRendererProperties}
   {rendererClass=&NiagaraSpriteRendererProperties, rendererIndex=1}
   {rendererClass=&NiagaraSpriteRendererProperties, rendererIndex=2}
  ]
 }
}
summary {
 systemName=NS_Emote_PiperShelf_Stomp
 emitters [
  {emitterName=Light_Shafts, rendererClasses [&NiagaraSpriteRendererProperties]}
  {emitterName=Sparkles_Glint, rendererClasses [&NiagaraSpriteRendererProperties]}
  {
   emitterName=Bright_Flash_Decal
   rendererClasses [&NiagaraSpriteRendererProperties, &NiagaraLightRendererProperties]
  }
 ]
 userVariables [
  User.bIsFrontEnd {type=NiagaraBool}
  User.bIsFrontEndPreview {type=NiagaraBool}
  User.BoneIndex {type=NiagaraFloat, defaultValue=0, description=""}
 ]
}
